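# Conversation database for the DashSurge NPC, defining the greeting and top-level branches when a player talks to this sushi-themed character.
# An entry point leads to one speech node with a default line and contextual variants for a first-ever meeting, per-match conversation count, or specific worn outfits.
# First-meeting and outfit reactions take curve-driven priorities.
# The speech node then branches through link nodes to the shared NPCLibrary quests and basic services conversations and DashSurge's own dialogue.
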ConversationDatabase /NPCLibrary/NPCs/DashSurge/NPCLibrary_Conversation_DashSurge
entry_points [{tag=Conversation.Entry.DashSurge, destinations [#1]}]
nodes [
 ConversationEntryPointNode_7 {
  id=#1
  class=ConversationEntryPointNode
  outputs [#2]
  properties {entryTag {tagName=Conversation.Entry.DashSurge}}
 }
 Speech_C_4 {
  id=#2
  class=Speech_C
  outputs [#3, #4, #5]
  properties {
   nodeName=""
   generalConfig {
    defaultMessage=Munch on my masterful maki.
    contextualMessages [
     {
      message="My rolls are Durrr-licious!"
      contextRequirements [{requirement {class=FortControllerRequirement_IsFirstEverConversationWithNPC}}]
      priority {curve {rowName=Tandem.Speech.Priority.FirstEverInteraction}}
     }
     {
      message=Munch on my masterful maki.
      contextRequirements [{requirement {class=FortControllerRequirement_CheckNPCMatchConversationCount, properties {nPCUniqueIdTag {tagName=AISpawnerData.Type.Tandem.DashSurge}}}}]
      priority {curve {rowName=None}}
     }
     {
      message="Fresh cut fish on rice. What could be better than that?!"
      contextRequirements [{requirement {class=FortControllerRequirement_CheckNPCMatchConversationCount, properties {nPCUniqueIdTag {tagName=AISpawnerData.Type.Tandem.DashSurge}}}}]
      priority {curve {rowName=None}}
     }
     {
      message="Don't be afraid, little fish! It impacts your deliciousness."
      contextRequirements [{requirement {
         class=FortControllerRequirement_HasCID
         properties {allowedCIds [&CID_315_Athena_Commando_M_TeriyakiFish]}
        }}]
      priority {curve {rowName=Tandem.Speech.Priority.OutfitReaction}}
     }
     {
      message="I'd be willing to franchise, if you're interested."
      contextRequirements [{requirement {
         class=FortControllerRequirement_HasCID
         properties {allowedCIds [/BRCosmetics/Athena/Items/Cosmetics/Characters/CID_271_Athena_Commando_F_SushiChef]}
        }}]
      priority {curve {rowName=Tandem.Speech.Priority.OutfitReaction}}
     }
     {
      message="Gila has thunder rolls. I have lightning rolls. They're completely different products!"
      contextRequirements [{requirement {class=FortControllerRequirement_HasCID, properties {allowedCIds [&Character_LocalZilla]}}}]
      priority {curve {rowName=Tandem.Speech.Priority.OutfitReaction}}
     }
    ]
   }
  }
 }
 ConversationLinkNode_2 {id=#5, class=ConversationLinkNode, properties {remoteEntryTag {tagName=Conversation.Entry.DashSurge.Dialogue}}}
 ConversationLinkNode_0 {id=#4, class=ConversationLinkNode, properties {remoteEntryTag {tagName=Conversation.Entry.BasicServices}}}
 ConversationLinkNode_1 {id=#3, class=ConversationLinkNode, properties {remoteEntryTag {tagName=Conversation.Entry.NPCLibrary.Quests}}}
]
